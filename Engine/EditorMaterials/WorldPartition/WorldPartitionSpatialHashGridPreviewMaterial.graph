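# Editor post-process material that previews World Partition spatial hash grids over the scene, centered on the camera position.
# Four duplicated chains, each parameterized by enable, cell size, loading range, offset and color, draw a range ring, cell borders and tinted in-range cells.
# Clamping the camera to cell bounds finds in-range cells; Is_Grid_2D drops height.
# The summed grids and camera marker are divided by eye adaptation and added to scene color.
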Material /Engine/EditorMaterials/WorldPartition/WorldPartitionSpatialHashGridPreviewMaterial
graph {
 properties {
  blendableLocation=BL_SceneColorAfterDOF
  blendableOutputAlpha=true
  materialDomain=MD_PostProcess
  pixelDepthOffsetMode=PDOM_Legacy
  usedShadingModels=Default Lit
 }
 nodes [
  n1 = MaterialFunctionCall(Float3=WorldPosition().XYZ) {materialFunction=&BreakOutFloat3Components}
  n2 = StaticSwitch(True=Constant(), False=Constant() {r=1}, Value=StaticBoolParameter() {defaultValue=true, parameterName=Is_Grid_2D})
  n3 = MaterialFunctionCall(Float3=CameraPositionWS()) {materialFunction=&BreakOutFloat3Components}
  n4 = SphereMask(
   A=MaterialFunctionCall(X=@n1.R, Y=@n1.R, Z=Multiply(A=@n2, B=@n1.B)) {materialFunction=&MakeFloat3}.Result
   B=MaterialFunctionCall(X=@n3.R, Y=@n3.R, Z=Multiply(A=@n2, B=@n3.B)) {materialFunction=&MakeFloat3}.Result
  ) {attenuationRadius=100, hardnessPercent=90}
  n5 = Multiply(A=@n4) {constB=5}
  n6 = Constant() {r=1}
  n7 = If(A=ScalarParameter() {parameterName=Grid0_Enabled, sliderMax=1}, B=@n6, "A > B"=@n6, "A == B"=@n6, "A < B"=Constant())
  n8 = ScalarParameter() {parameterName=Grid0_LoadingRange, sliderMax=102400, sliderMin=1600}
  n9 = ScalarParameter() {defaultValue=3200, parameterName=Grid0_CellSize, sliderMax=102400, sliderMin=1600}
  n10 = Subtract(
   A=Subtract(A=WorldPosition().XYZ, B=Multiply(A=@n9) {constB=0.5})
   B=VectorParameter() {parameterName=Grid0_Offset, defaultValue=<rgba 0 0 0 1>}.RGB
  )
  n11 = MaterialFunctionCall(Float3=@n10) {materialFunction=&BreakOutFloat3Components}
  n12 = StaticSwitch(True=Constant(), False=Constant() {r=1}, Value=StaticBoolParameter() {defaultValue=true, parameterName=Is_Grid_2D})
  n13 = MaterialFunctionCall(X=@n11.R, Y=@n11.R, Z=Multiply(A=@n12, B=@n11.B)) {materialFunction=&MakeFloat3}
  n14 = CameraPositionWS()
  n15 = Multiply(A=ScalarParameter() {defaultValue=3200, parameterName=Grid0_CellSize, sliderMax=102400, sliderMin=1600}) {constB=0.5}
  n16 = MaterialFunctionCall(Float3=Subtract(A=Subtract(A=@n14, B=@n15), B=VectorParameter() {parameterName=Grid0_Offset, defaultValue=<rgba 0 0 0 1>}.RGB)) {materialFunction=&BreakOutFloat3Components}
  n17 = MaterialFunctionCall(X=@n16.R, Y=@n16.R, Z=Multiply(A=@n12, B=@n16.B)) {materialFunction=&MakeFloat3}
  n18 = Distance(A=@n17.Result, B=@n13.Result)
  n19 = Constant() {r=1}
  n20 = Constant()
  n21 = Subtract(A=@n8) {constB=100}
  n22 = ScalarParameter() {defaultValue=3200, parameterName=Grid0_CellSize, sliderMax=102400, sliderMin=1600}
  n23 = Multiply(A=Floor(None=Divide(A=@n13.Result, B=@n22)), B=@n22)
  n24 = Clamp(None=@n17.Result, Min=@n23, Max=Add(A=@n23, B=@n22))
  n25 = Constant() {r=1}
  n26 = If(
   A=ScalarParameter() {parameterName=Grid0_LoadingRange, sliderMax=102400, sliderMin=1600}
   B=Distance(A=Subtract(A=@n17.Result, B=@n24), B=Constant())
   "A > B"=@n25
   "A == B"=@n25
   "A < B"=Constant()
  ) {constB=102400}
  n27 = Constant() {r=0.25}
  n28 = ScalarParameter() {defaultValue=3200, parameterName=Grid0_CellSize, sliderMax=102400, sliderMin=1600}
  n29 = Subtract(A=Fmod(A=Abs(None=@n10), B=@n28), B=Divide(A=@n28))
  n30 = Add(A=Abs(None=Multiply(A=@n29) {constB=2})) {constB=100}
  n31 = MaterialFunctionCall(Float3=Saturate(None=Subtract(A=@n30, B=@n28) {constB=100})) {materialFunction=&BreakOutFloat3Components}
  n32 = Constant()
  n33 = StaticBoolParameter() {defaultValue=true, parameterName=Is_Grid_2D}
  n34 = Saturate(None=Add(A=Add(A=@n31.R, B=@n31.R), B=StaticSwitch(True=@n32, False=@n31.B, Value=@n33)))
  n35 = Multiply(A=@n26, B=@n34)
  n36 = Saturate(None=Subtract(A=Multiply(A=@n26, B=@n27), B=@n35))
  n37 = Multiply(A=@n8) {constB=3}
  n38 = Multiply(
   A=Multiply(A=SphereMask(A=@n13.Result, B=@n17.Result, Radius=@n37) {attenuationRadius=100, hardnessPercent=0}) {constB=0.5}
   B=OneMinus(None=@n26)
  )
  n39 = Add(
   A=Multiply(
    A=If(A=@n8, B=@n18, "A > B"=@n19, "A == B"=@n20, "A < B"=@n20) {constB=102400}
    B=If(A=@n21, B=@n18, "A > B"=@n20, "A == B"=@n19, "A < B"=@n19) {constB=102400}
   ) {constB=0.5}
   B=Add(A=Add(A=@n36, B=@n35), B=Multiply(A=@n38, B=@n34))
  )
  n40 = MaterialFunctionCall(Float3=Multiply(A=Multiply(A=@n7, B=@n39), B=VectorParameter() {parameterName=Grid0_Color, defaultValue=<rgba 0 20 11 0>}.RGB)) {materialFunction=&BreakOutFloat3Components}
  n41 = Constant() {r=1}
  n42 = If(A=ScalarParameter() {parameterName=Grid1_Enabled, sliderMax=1}, B=@n41, "A > B"=@n41, "A == B"=@n41, "A < B"=Constant())
  n43 = ScalarParameter() {parameterName=Grid1_LoadingRange, sliderMax=102400, sliderMin=1600}
  n44 = ScalarParameter() {defaultValue=3200, parameterName=Grid1_CellSize, sliderMax=102400, sliderMin=1600}
  n45 = Subtract(
   A=Subtract(A=WorldPosition().XYZ, B=Multiply(A=@n44) {constB=0.5})
   B=VectorParameter() {parameterName=Grid1_Offset, defaultValue=<rgba 0 0 0 1>}.RGB
  )
  n46 = MaterialFunctionCall(Float3=@n45) {materialFunction=&BreakOutFloat3Components}
  n47 = StaticSwitch(True=Constant(), False=Constant() {r=1}, Value=StaticBoolParameter() {defaultValue=true, parameterName=Is_Grid_2D})
  n48 = MaterialFunctionCall(X=@n46.R, Y=@n46.R, Z=Multiply(A=@n47, B=@n46.B)) {materialFunction=&MakeFloat3}
  n49 = CameraPositionWS()
  n50 = Multiply(A=ScalarParameter() {defaultValue=3200, parameterName=Grid1_CellSize, sliderMax=102400, sliderMin=1600}) {constB=0.5}
  n51 = MaterialFunctionCall(Float3=Subtract(A=Subtract(A=@n49, B=@n50), B=VectorParameter() {parameterName=Grid1_Offset, defaultValue=<rgba 0 0 0 1>}.RGB)) {materialFunction=&BreakOutFloat3Components}
  n52 = MaterialFunctionCall(X=@n51.R, Y=@n51.R, Z=Multiply(A=@n47, B=@n51.B)) {materialFunction=&MakeFloat3}
  n53 = Distance(A=@n52.Result, B=@n48.Result)
  n54 = Constant() {r=1}
  n55 = Constant()
  n56 = Subtract(A=@n43) {constB=100}
  n57 = ScalarParameter() {defaultValue=3200, parameterName=Grid1_CellSize, sliderMax=102400, sliderMin=1600}
  n58 = Multiply(A=Floor(None=Divide(A=@n48.Result, B=@n57)), B=@n57)
  n59 = Clamp(None=@n52.Result, Min=@n58, Max=Add(A=@n58, B=@n57))
  n60 = Constant() {r=1}
  n61 = If(
   A=ScalarParameter() {parameterName=Grid1_LoadingRange, sliderMax=102400, sliderMin=1600}
   B=Distance(A=Subtract(A=@n52.Result, B=@n59), B=Constant())
   "A > B"=@n60
   "A == B"=@n60
   "A < B"=Constant()
  ) {constB=102400}
  n62 = Constant() {r=0.25}
  n63 = ScalarParameter() {defaultValue=3200, parameterName=Grid1_CellSize, sliderMax=102400, sliderMin=1600}
  n64 = Subtract(A=Fmod(A=Abs(None=@n45), B=@n63), B=Divide(A=@n63))
  n65 = Add(A=Abs(None=Multiply(A=@n64) {constB=2})) {constB=100}
  n66 = MaterialFunctionCall(Float3=Saturate(None=Subtract(A=@n65, B=@n63) {constB=100})) {materialFunction=&BreakOutFloat3Components}
  n67 = Constant()
  n68 = StaticBoolParameter() {defaultValue=true, parameterName=Is_Grid_2D}
  n69 = Saturate(None=Add(A=Add(A=@n66.R, B=@n66.R), B=StaticSwitch(True=@n67, False=@n66.B, Value=@n68)))
  n70 = Multiply(A=@n61, B=@n69)
  n71 = Saturate(None=Subtract(A=Multiply(A=@n61, B=@n62), B=@n70))
  n72 = Multiply(A=@n43) {constB=3}
  n73 = Multiply(
   A=Multiply(A=SphereMask(A=@n48.Result, B=@n52.Result, Radius=@n72) {attenuationRadius=100, hardnessPercent=0}) {constB=0.5}
   B=OneMinus(None=@n61)
  )
  n74 = Add(
   A=Multiply(
    A=If(A=@n43, B=@n53, "A > B"=@n54, "A == B"=@n55, "A < B"=@n55) {constB=102400}
    B=If(A=@n56, B=@n53, "A > B"=@n55, "A == B"=@n54, "A < B"=@n54) {constB=102400}
   ) {constB=0.5}
   B=Add(A=Add(A=@n71, B=@n70), B=Multiply(A=@n73, B=@n69))
  )
  n75 = MaterialFunctionCall(Float3=Multiply(A=Multiply(A=@n42, B=@n74), B=VectorParameter() {parameterName=Grid1_Color, defaultValue=<rgba 20 0 2.0719 0>}.RGB)) {materialFunction=&BreakOutFloat3Components}
  n76 = Add(
   A=MaterialFunctionCall(X=@n40.R, Y=@n40.R, Z=@n40.R) {materialFunction=&MakeFloat4}.Result
   B=MaterialFunctionCall(X=@n75.R, Y=@n75.R, Z=@n75.R) {materialFunction=&MakeFloat4}.Result
  )
  n77 = Constant() {r=1}
  n78 = If(A=ScalarParameter() {parameterName=Grid2_Enabled, sliderMax=1}, B=@n77, "A > B"=@n77, "A == B"=@n77, "A < B"=Constant())
  n79 = ScalarParameter() {parameterName=Grid2_LoadingRange, sliderMax=102400, sliderMin=1600}
  n80 = ScalarParameter() {defaultValue=3200, parameterName=Grid2_CellSize, sliderMax=102400, sliderMin=1600}
  n81 = Subtract(
   A=Subtract(A=WorldPosition().XYZ, B=Multiply(A=@n80) {constB=0.5})
   B=VectorParameter() {parameterName=Grid2_Offset, defaultValue=<rgba 0 0 0 1>}.RGB
  )
  n82 = MaterialFunctionCall(Float3=@n81) {materialFunction=&BreakOutFloat3Components}
  n83 = StaticSwitch(True=Constant(), False=Constant() {r=1}, Value=StaticBoolParameter() {defaultValue=true, parameterName=Is_Grid_2D})
  n84 = MaterialFunctionCall(X=@n82.R, Y=@n82.R, Z=Multiply(A=@n83, B=@n82.B)) {materialFunction=&MakeFloat3}
  n85 = CameraPositionWS()
  n86 = Multiply(A=ScalarParameter() {defaultValue=3200, parameterName=Grid2_CellSize, sliderMax=102400, sliderMin=1600}) {constB=0.5}
  n87 = MaterialFunctionCall(Float3=Subtract(A=Subtract(A=@n85, B=@n86), B=VectorParameter() {parameterName=Grid2_Offset, defaultValue=<rgba 0 0 0 1>}.RGB)) {materialFunction=&BreakOutFloat3Components}
  n88 = MaterialFunctionCall(X=@n87.R, Y=@n87.R, Z=Multiply(A=@n83, B=@n87.B)) {materialFunction=&MakeFloat3}
  n89 = Distance(A=@n88.Result, B=@n84.Result)
  n90 = Constant() {r=1}
  n91 = Constant()
  n92 = Subtract(A=@n79) {constB=100}
  n93 = ScalarParameter() {defaultValue=3200, parameterName=Grid2_CellSize, sliderMax=102400, sliderMin=1600}
  n94 = Multiply(A=Floor(None=Divide(A=@n84.Result, B=@n93)), B=@n93)
  n95 = Clamp(None=@n88.Result, Min=@n94, Max=Add(A=@n94, B=@n93))
  n96 = Constant() {r=1}
  n97 = If(
   A=ScalarParameter() {parameterName=Grid2_LoadingRange, sliderMax=102400, sliderMin=1600}
   B=Distance(A=Subtract(A=@n88.Result, B=@n95), B=Constant())
   "A > B"=@n96
   "A == B"=@n96
   "A < B"=Constant()
  ) {constB=102400}
  n98 = Constant() {r=0.25}
  n99 = ScalarParameter() {defaultValue=3200, parameterName=Grid2_CellSize, sliderMax=102400, sliderMin=1600}
  n100 = Subtract(A=Fmod(A=Abs(None=@n81), B=@n99), B=Divide(A=@n99))
  n101 = Add(A=Abs(None=Multiply(A=@n100) {constB=2})) {constB=100}
  n102 = MaterialFunctionCall(Float3=Saturate(None=Subtract(A=@n101, B=@n99) {constB=100})) {materialFunction=&BreakOutFloat3Components}
  n103 = Constant()
  n104 = StaticBoolParameter() {defaultValue=true, parameterName=Is_Grid_2D}
  n105 = Saturate(None=Add(A=Add(A=@n102.R, B=@n102.R), B=StaticSwitch(True=@n103, False=@n102.B, Value=@n104)))
  n106 = Multiply(A=@n97, B=@n105)
  n107 = Saturate(None=Subtract(A=Multiply(A=@n97, B=@n98), B=@n106))
  n108 = Multiply(A=@n79) {constB=3}
  n109 = Multiply(
   A=Multiply(A=SphereMask(A=@n84.Result, B=@n88.Result, Radius=@n108) {attenuationRadius=100, hardnessPercent=0}) {constB=0.5}
   B=OneMinus(None=@n97)
  )
  n110 = Add(
   A=Multiply(
    A=If(A=@n79, B=@n89, "A > B"=@n90, "A == B"=@n91, "A < B"=@n91) {constB=102400}
    B=If(A=@n92, B=@n89, "A > B"=@n91, "A == B"=@n90, "A < B"=@n90) {constB=102400}
   ) {constB=0.5}
   B=Add(A=Add(A=@n107, B=@n106), B=Multiply(A=@n109, B=@n105))
  )
  n111 = MaterialFunctionCall(Float3=Multiply(A=Multiply(A=@n78, B=@n110), B=VectorParameter() {parameterName=Grid2_Color, defaultValue=<rgba 20 17.4986 0 0>}.RGB)) {materialFunction=&BreakOutFloat3Components}
  n112 = MaterialFunctionCall(X=@n111.R, Y=@n111.R, Z=@n111.R) {materialFunction=&MakeFloat4}
  n113 = Constant() {r=1}
  n114 = If(A=ScalarParameter() {parameterName=Grid3_Enabled, sliderMax=1}, B=@n113, "A > B"=@n113, "A == B"=@n113, "A < B"=Constant())
  n115 = ScalarParameter() {parameterName=Grid3_LoadingRange, sliderMax=102400, sliderMin=1600}
  n116 = ScalarParameter() {defaultValue=3200, parameterName=Grid3_CellSize, sliderMax=102400, sliderMin=1600}
  n117 = Subtract(
   A=Subtract(A=WorldPosition().XYZ, B=Multiply(A=@n116) {constB=0.5})
   B=VectorParameter() {parameterName=Grid3_Offset, defaultValue=<rgba 0 0 0 1>}.RGB
  )
  n118 = MaterialFunctionCall(Float3=@n117) {materialFunction=&BreakOutFloat3Components}
  n119 = StaticSwitch(True=Constant(), False=Constant() {r=1}, Value=StaticBoolParameter() {defaultValue=true, parameterName=Is_Grid_2D})
  n120 = MaterialFunctionCall(X=@n118.R, Y=@n118.R, Z=Multiply(A=@n119, B=@n118.B)) {materialFunction=&MakeFloat3}
  n121 = CameraPositionWS()
  n122 = Multiply(A=ScalarParameter() {defaultValue=3200, parameterName=Grid3_CellSize, sliderMax=102400, sliderMin=1600}) {constB=0.5}
  n123 = MaterialFunctionCall(Float3=Subtract(A=Subtract(A=@n121, B=@n122), B=VectorParameter() {parameterName=Grid3_Offset, defaultValue=<rgba 0 0 0 1>}.RGB)) {materialFunction=&BreakOutFloat3Components}
  n124 = MaterialFunctionCall(X=@n123.R, Y=@n123.R, Z=Multiply(A=@n119, B=@n123.B)) {materialFunction=&MakeFloat3}
  n125 = Distance(A=@n124.Result, B=@n120.Result)
  n126 = Constant() {r=1}
  n127 = Constant()
  n128 = Subtract(A=@n115) {constB=100}
  n129 = ScalarParameter() {defaultValue=3200, parameterName=Grid3_CellSize, sliderMax=102400, sliderMin=1600}
  n130 = Multiply(A=Floor(None=Divide(A=@n120.Result, B=@n129)), B=@n129)
  n131 = Clamp(None=@n124.Result, Min=@n130, Max=Add(A=@n130, B=@n129))
  n132 = Constant() {r=1}
  n133 = If(
   A=ScalarParameter() {parameterName=Grid3_LoadingRange, sliderMax=102400, sliderMin=1600}
   B=Distance(A=Subtract(A=@n124.Result, B=@n131), B=Constant())
   "A > B"=@n132
   "A == B"=@n132
   "A < B"=Constant()
  ) {constB=102400}
  n134 = Constant() {r=0.25}
  n135 = ScalarParameter() {defaultValue=3200, parameterName=Grid3_CellSize, sliderMax=102400, sliderMin=1600}
  n136 = Subtract(A=Fmod(A=Abs(None=@n117), B=@n135), B=Divide(A=@n135))
  n137 = Add(A=Abs(None=Multiply(A=@n136) {constB=2})) {constB=100}
  n138 = MaterialFunctionCall(Float3=Saturate(None=Subtract(A=@n137, B=@n135) {constB=100})) {materialFunction=&BreakOutFloat3Components}
  n139 = Constant()
  n140 = StaticBoolParameter() {defaultValue=true, parameterName=Is_Grid_2D}
  n141 = Saturate(None=Add(A=Add(A=@n138.R, B=@n138.R), B=StaticSwitch(True=@n139, False=@n138.B, Value=@n140)))
  n142 = Multiply(A=@n133, B=@n141)
  n143 = Saturate(None=Subtract(A=Multiply(A=@n133, B=@n134), B=@n142))
  n144 = Multiply(A=@n115) {constB=3}
  n145 = Multiply(
   A=Multiply(A=SphereMask(A=@n120.Result, B=@n124.Result, Radius=@n144) {attenuationRadius=100, hardnessPercent=0}) {constB=0.5}
   B=OneMinus(None=@n133)
  )
  n146 = Add(
   A=Multiply(
    A=If(A=@n115, B=@n125, "A > B"=@n126, "A == B"=@n127, "A < B"=@n127) {constB=102400}
    B=If(A=@n128, B=@n125, "A > B"=@n127, "A == B"=@n126, "A < B"=@n126) {constB=102400}
   ) {constB=0.5}
   B=Add(A=Add(A=@n143, B=@n142), B=Multiply(A=@n145, B=@n141))
  )
  n147 = MaterialFunctionCall(Float3=Multiply(A=Multiply(A=@n114, B=@n146), B=VectorParameter() {parameterName=Grid3_Color, defaultValue=<rgba 0.1319 20 0 0>}.RGB)) {materialFunction=&BreakOutFloat3Components}
  n148 = Multiply(
   A=OneMinus(None=@n4)
   B=Add(
    A=Add(A=@n76, B=@n112.Result)
    B=MaterialFunctionCall(X=@n147.R, Y=@n147.R, Z=@n147.R) {materialFunction=&MakeFloat4}.Result
   )
  )
  Add(A=Divide(A=Add(A=@n5, B=@n148), B=EyeAdaptation().EyeAdaptation), B=SceneTexture() {sceneTextureId=PPI_PostProcessInput0}.Color)
 ]
}
